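# Translucent unlit Niagara material for animated line bands in the Polarity Winn glider trail.
# Time-scrolled sine waves along V form exponent-sharpened lines, alongside vertical gradient, upper detail and a border fade driven by UVs, vertex color and dynamic Appear.
# A three-color lerp and particle color feed emissive adjustments.
# A WPO branch builds a clamped offset from dynamic X minus particle position, weighted by V and WPOIntensity, only on High and Epic quality.
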Material /BRCosmetics/Gadgets/Assets/VinderTech_GliderChute/Glider_PolarityWinn/FX/Trail/M_PolarityWinn_Lines
graph {
 properties {
  bUsedWithNiagaraMeshParticles=true
  bUsedWithNiagaraSprites=true
  blendMode=BLEND_Translucent
  pixelDepthOffsetMode=PDOM_Legacy
  shadingModel=MSM_Unlit
  usedShadingModels=Unlit
 }
 nodes [
  n1 = NamedRerouteUsage()
  n2 = ScalarParameter() {parameterName=LinesColorGradientBias}
  n3 = MaterialFunctionCall(
   A=VectorParameter() {parameterName=LinesColor1, defaultValue=<rgba 0.3609 0 1 1>}.RGB
   B=VectorParameter() {parameterName=LinesColor2, defaultValue=<rgba 0.089 0.1589 1 0.5>}.RGB
   C=VectorParameter() {parameterName=LinesColor3, defaultValue=<rgba 0 1.0625 2 1>}.RGB
   Alpha=Saturate(None=Add(A=@n1, B=@n2) {constB=0})
  ) {materialFunction=&Lerp_3Color}
  n4 = NamedRerouteUsage()
  n5 = ParticleColor()
  MaterialFunctionCall(Emissive=Multiply(A=Multiply(A=@n3.Result, B=@n4), B=@n5.RGB)) {materialFunction=&MF_FX_EmissiveAdjustments}
  n6 = TextureCoordinate()
  n7 = VertexInterpolator(VS=@n6)
  n8 = ComponentMask(None=@n7.PS) {g=true}
  n9 = Multiply(A=Power(Base=@n8) {constExponent=1.5}, B=OneMinus(None=@n8))
  Multiply(A=@n9) {constB=3}
  Multiply(A=Multiply(A=@n4, B=NamedRerouteUsage()), B=@n5.A)
  n10 = OneMinus(None=NamedRerouteUsage())
  Multiply(A=Power(Base=Saturate(None=@n10)) {constExponent=12}, B=@n4)
  n11 = ComponentMask(None=TextureCoordinate()) {g=true}
  n12 = ScalarParameter() {defaultValue=2, parameterName=LineExp_Gradient}
  NamedRerouteDeclaration(None=Multiply(A=Power(Base=@n11, Exp=@n12) {constExponent=1.5}, B=ScalarParameter() {defaultValue=1, parameterName=LineMult_Gradient})) {name=Vertical Gradient, nodeColor=<rgba 0.5234 0 1 1>}
  n13 = ComponentMask(None=@n7.PS) {r=true}
  n14 = Power(Base=@n13) {constExponent=1.5}
  n15 = OneMinus(None=@n13)
  n16 = Multiply(A=Multiply(A=Multiply(A=@n14, B=@n15), B=@n9) {constB=2}) {constB=25}
  n17 = ComponentMask(None=@n6) {g=true}
  n18 = Multiply(
   A=Power(Base=VertexColor()) {constExponent=4}
   B=SmoothStep(Min=DynamicParameter() {defaultValue=<rgba 0 -0.1 1 1>, paramNames [Appear,Param2,Param3,Param4]}.Appear, Value=OneMinus(None=@n17)) {constMin=0.9, constValue=0.5}
  )
  NamedRerouteDeclaration(None=Saturate(None=Multiply(A=@n16, B=@n18))) {name=Border Fade, nodeColor=<rgba 1 0.2109 0 1>}
  n19 = Constant()
  n20 = DynamicParameter() {parameterIndex=1, paramNames [X,Y,Z,WPOIntensity]}
  n21 = Subtract(
   A=MaterialFunctionCall(R=@n20.X, G=@n20.X, B=@n20.X) {materialFunction=&AppendMany}.RGB
   B=ParticlePositionWS()
  )
  n22 = ComponentMask(None=@n21) {r=true}
  n23 = Constant() {r=-50}
  n24 = Constant() {r=50}
  n25 = ComponentMask(None=@n21) {g=true}
  n26 = Constant() {r=-50}
  n27 = Constant() {r=50}
  n28 = ComponentMask(None=@n21) {b=true}
  n29 = Constant() {r=-10}
  n30 = Constant() {r=10}
  n31 = ComponentMask(None=TextureCoordinate()) {g=true}
  n32 = ScalarParameter() {defaultValue=2, parameterName=Interpolate Blend}
  n33 = LinearInterpolate(
   B=MaterialFunctionCall(X=Clamp(None=@n22, Min=@n23, Max=@n24), Y=Clamp(None=@n25, Min=@n26, Max=@n27), Z=Clamp(None=@n28, Min=@n29, Max=@n30)) {materialFunction=&MakeFloat3}.Result
   Alpha=Multiply(A=@n20.WPOIntensity, B=Power(Base=@n31, Exp=@n32))
  )
  NamedRerouteDeclaration(None=QualitySwitch(Default=@n19, Low=@n19, High=@n33, Medium=@n19, Epic=@n33)) {name=WPO, nodeColor=<rgba 0 0.3437 1 1>}
  n34 = ComponentMask(None=TextureCoordinate()) {g=true}
  n35 = Time()
  n36 = ScalarParameter() {defaultValue=0.2, parameterName=TimeSpeed}
  n37 = Add(
   A=NamedRerouteDeclaration(None=OneMinus(None=@n34)) {name=Gradient, nodeColor=<rgba 0 0.1094 1 1>}
   B=NamedRerouteDeclaration(None=Multiply(A=@n35, B=@n36) {constB=0.2}) {name=Time, nodeColor=<rgba 1 0.0469 0 1>}
  )
  n38 = ScalarParameter() {defaultValue=6, parameterName=Line Number}
  n39 = ConstantBiasScale(None=Sine(None=Multiply(A=@n37, B=@n38) {constB=7}))
  n40 = Multiply(A=ScalarParameter() {defaultValue=2, parameterName=LineExponent}, B=NamedRerouteUsage())
  NamedRerouteDeclaration(None=Saturate(None=Power(Base=@n39, Exp=@n40) {constExponent=6})) {name=SineWaves, nodeColor=<rgba 0.8281 0 1 1>}
  n41 = ConstantBiasScale(None=NamedRerouteUsage())
  n42 = Power(Base=OneMinus(None=NamedRerouteUsage()), Exp=Subtract(A=Constant() {r=100}, B=Multiply(A=@n41) {constB=50})) {constExponent=64}
  NamedRerouteDeclaration(None=Multiply(A=Saturate(None=@n42), B=Constant() {r=1}) {constB=4}) {name=UpperDetail, nodeColor=<rgba 1 0.1875 0 1>}
  NamedRerouteUsage()
  OneMinus(None=@n41)
  QualitySwitch()
 ]
}
